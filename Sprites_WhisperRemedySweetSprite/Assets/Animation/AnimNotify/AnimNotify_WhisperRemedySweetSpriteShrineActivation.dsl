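; Anim notify state for the Whisper Remedy Sweet sprite, named for its shrine activation, that swaps the animated mesh's materials.
; When the notify begins, it sets the body and facial material instances on the first two slots and a global VertexCrunch overlay material on the third.
; The end event is empty, so nothing is reverted and the materials persist past the notify window.
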(fn Received_NotifyBegin
  (params (MeshComp :type Skeletal%20Mesh%20Component%20Object%20Reference) (Animation :type Anim%20Sequence%20Base%20Object%20Reference) (TotalDuration :type Float%20%28single-precision%29) (EventReference :type Anim%20Notify%20Event%20Reference%20Structure%20%28by%20ref%29))
  (Rendering|Material|SetMaterial MeshComp 0 "/Sprites_WhisperRemedySweetSprite/Assets/Materials/MI_WhisperRemedySweet_Body.MI_WhisperRemedySweet_Body")
  (Rendering|Material|SetMaterial MeshComp 1 "/Sprites_WhisperRemedySweetSprite/Assets/Materials/MI_Sprite_WhisperRemedySweet_Facial.MI_Sprite_WhisperRemedySweet_Facial")
  (Rendering|Material|SetMaterial MeshComp 2 "/Game/Global/Materials/M_Overlay_VertexCrunch.M_Overlay_VertexCrunch"))

(fn Received_NotifyEnd
  (params (MeshComp :type Skeletal%20Mesh%20Component%20Object%20Reference) (Animation :type Anim%20Sequence%20Base%20Object%20Reference) (EventReference :type Anim%20Notify%20Event%20Reference%20Structure%20%28by%20ref%29)))
