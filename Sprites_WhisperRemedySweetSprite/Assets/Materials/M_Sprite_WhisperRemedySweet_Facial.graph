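# Unlit skeletal-mesh material for the WhisperRemedySweet sprite's face, with reactive glowing red eyes.
# Pre-skinned height gradient and vertex-color red form the base brightness.
# A radial glow around EyeCenter and its X-mirror blends two reactive colors, masked over tinted base by the EyeWhite texture.
# The reaction fades after CoolDownTime unless IsInstantRedEye holds it, with eye-adaptation compensation; separate nodes compute a scale-aware camera push offset.
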Material /Sprites_WhisperRemedySweetSprite/Assets/Materials/M_Sprite_WhisperRemedySweet_Facial
graph {
 properties {
  bAutomaticallySetUsageInEditor=false
  bUsedWithMeshDeformer=true
  bUsedWithSkeletalMesh=true
  shadingModel=MSM_Unlit
  usedShadingModels=Unlit
 }
 nodes [
  Constant() {r=0.1}
  n1 = LocalPosition() {localOrigin=InstancePreSkinning}
  n2 = ShaderStageSwitch(PixelShader=VertexInterpolator(VS=@n1.Z).PS, VertexShader=@n1.Z)
  n3 = VertexColor()
  n4 = LinearInterpolate(A=SmoothStep(Value=Divide(A=@n2) {constB=50}) {constMax=0.6, constMin=0.42}, Alpha=@n3.R) {constB=5}
  n5 = Multiply(A=@n4, B=VectorParameter() {parameterName=Eye_Red_HighLight, defaultValue=<rgba 1 0.3719 0.3719 1>}.RGB)
  n6 = VectorParameter() {group=Reactivity, parameterName=EyeReactiveColor, defaultValue=<rgba 1 0 0 5>}
  n7 = VectorParameter() {group=Reactivity, parameterName=EyeReactiveColor_Center, defaultValue=<rgba 1 0.2554 0 5>}
  n8 = VertexInterpolator(VS=PreSkinnedPosition())
  n9 = VectorParameter() {parameterName=EyeCenter, defaultValue=<rgba 0 0 0 1>}
  n10 = Length(None=Subtract(A=@n8.PS, B=@n9.RGB)) {is=class MaterialExpressionLength}
  n11 = ScalarParameter() {defaultValue=5, parameterName=EyeRadius}
  n12 = OneMinus(None=Saturate(None=Divide(A=@n10, B=@n11)))
  n13 = Constant3Vector() {constant=<rgba -1 1 1 0>}
  n14 = Length(None=Subtract(A=Multiply(A=@n9.RGB, B=@n13), B=@n8.PS))
  n15 = OneMinus(None=Saturate(None=Divide(A=@n14, B=@n11)))
  n16 = LinearInterpolate(A=Multiply(A=@n6.RGB, B=@n6.RGB), B=Multiply(A=@n7.RGB, B=@n7.RGB), Alpha=Power(Base=Add(A=@n12, B=@n15)) {constExponent=3})
  n17 = TextureSampleParameter2D() {parameterName=EyeWhite, texture=/Sprites_WhisperRemedySweetSprite/Assets/Textures/T_Sprite_EYE_Superman}
  n18 = Subtract(A=Time(), B=ScalarParameter() {group=Reactivity, parameterName=CoolDownTime})
  n19 = OneMinus(None=Saturate(None=Divide(A=@n18) {constB=1.5}))
  n20 = Constant() {r=1}
  n21 = ScalarParameter() {group=Reactivity, parameterName=IsInstantRedEye}
  Divide(
   A=LinearInterpolate(A=@n4, B=LinearInterpolate(A=@n5, B=@n16, Alpha=@n17.RGB), Alpha=LinearInterpolate(A=@n19, B=@n20, Alpha=@n21))
   B=EyeAdaptation().EyeAdaptation
  )
  Multiply(
   A=Multiply(A=MaterialFunctionCall() {materialFunction=&CameraDirectionVector}.Result) {constB=-1}
   B=Multiply(
    A=ScalarParameter() {parameterName=WPO_PushToCameraAmount}
    B=MaterialFunctionCall() {materialFunction=&ObjectScale}.Scale Z
   )
  )
  Multiply(A=@n3.R)
  ScalarParameter() {defaultValue=1, group=Reactivity, parameterName=RedEye_Alpha}
 ]
}
