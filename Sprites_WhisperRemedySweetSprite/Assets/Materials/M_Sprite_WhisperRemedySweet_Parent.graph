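# Parent subsurface material for the Whisper Remedy Sweet Sprite character, blending a procedural glowing body with textured surfaces through a mask.
# The glow layers height-based gradients, sphere-mask cores, triplanar virtual-box noise and sun scattering, with time-of-day dimming and an optional logo.
# Textured areas use SRM and normal maps, a metal LUT function and masked emissive.
# Wave displacement and bubble-normal networks exist, but the glow layer takes constant offset and normal.
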Material /Sprites_WhisperRemedySweetSprite/Assets/Materials/M_Sprite_WhisperRemedySweet_Parent
graph {
 properties {
  bAutomaticallySetUsageInEditor=false
  bUseMaterialAttributes=true
  bUsedWithMeshDeformer=true
  bUsedWithSkeletalMesh=true
  shadingModel=MSM_Subsurface
  usedShadingModels=Subsurface
 }
 nodes [
  n1 = ComponentMask(None=NamedRerouteUsage()) {b=true, g=true, r=true}
  n2 = NamedRerouteUsage()
  n3 = ScalarParameter() {defaultValue=0.7, group=Color_Shading, parameterName=InnerGlowRadius_Core}
  n4 = ScalarParameter() {group=Color_Shading, parameterName=InnerGlowHardness_Core}
  n5 = Power(Base=SphereMask(A=ComponentMask(None=@n1) {b=true, r=true}, B=ComponentMask(None=@n2) {b=true, r=true}, Radius=@n3, Hardness=@n4) {attenuationRadius=12, hardnessPercent=55})
  n6 = ComponentMask(None=@n1) {b=true, g=true}
  n7 = ComponentMask(None=@n2) {b=true, g=true}
  Add(A=@n5, B=Power(Base=SphereMask(A=@n6, B=@n7, Radius=@n3, Hardness=@n4) {attenuationRadius=12, hardnessPercent=55}))
  Add()
  n8 = VertexNormalWS()
  Add(A=AppendVector(A=NamedRerouteUsage(), B=Constant()), B=@n8)
  n9 = VectorParameter() {group=UV, parameterName=Zero_03_Speed, sortPriority=64, defaultValue=<rgba 0 0.2 0 1>}
  n10 = ComponentMask(None=MaterialFunctionCall!(
    Texture=TextureObject() {samplerType=SAMPLERTYPE_Masks, texture=/Sprites_WhisperRemedySweetSprite/Assets/Textures/T_SimpleDirct}
    UseBottomLayer=StaticBool()
    Depth=ScalarParameter() {defaultValue=10, group=UV, parameterName=Zero_03_Depth, sortPriority=61}
    Depth_Z=ScalarParameter() {defaultValue=25, group=UV, parameterName=Zero_03_Depth_Z, sortPriority=61}
    Tile=ScalarParameter() {defaultValue=1, group=UV, parameterName=Zero_03_Tile, sortPriority=62}
    BlendContrast_Z
    BlendContrast=ScalarParameter() {defaultValue=0.4, group=UV, parameterName=Zero_03_BlendContrast, sortPriority=63}
    SpeedXY=ComponentMask(None=@n9.RGB) {g=true, r=true}
    ActorOffset=VectorParameter() {group=UV, parameterName=Zero_03_Offset, sortPriority=64, defaultValue=<rgba 0 0 20 1>}.RGB
    StaticBottomLayer
    UseLocalCameraVector
    Distort
   ) {materialFunction=&/MF_TriplanarVirtualBox_WhisperRemedySweet}.Result) {g=true, r=true}
  AppendVector(A=Constant(), B=@n10)
  n11 = NamedRerouteUsage()
  n12 = SetMaterialAttributes!(
   MaterialAttributes
   Base Color=Constant()
   Metallic=Constant() {r=1}
   Specular=Constant() {r=1}
   Roughness=Constant() {r=0.25}
   Normal=Constant3Vector() {constant=<rgba 0 0 1 0>}
   Emissive Color=NamedRerouteUsage()
   Opacity
   Subsurface Color=@n11
   World Position Offset=Constant()
  ) {attributeSetTypes [
    #1
    #2
    #3
    #4
    #5
    #6
    #7
    #8
    #9
   ]}
  n13 = TextureSampleParameter2D() {group=SuperOverride, parameterName=Mask, samplerType=SAMPLERTYPE_Masks, texture=&T_BlackMask}
  n14 = TextureSampleParameter2D() {group=SuperOverride, parameterName=SRM, texture=/Sprites_WhisperRemedySweetSprite/Assets/Textures/T_Sprite_WhisperRemedySweet_Body_S}
  n15 = TextureSampleParameter2D() {
   group=SuperOverride
   parameterName=Normal
   samplerType=SAMPLERTYPE_Normal
   texture=/Sprites_WhisperRemedySweetSprite/Assets/Textures/T_Sprite_WhisperRemedySweet_Body_N
  }
  n16 = SetMaterialAttributes!(
   MaterialAttributes
   Base Color=TextureSampleParameter2D() {group=SuperOverride, parameterName=BaseColor, texture=&T_BlacksRGB}.RGB
   Metallic=StaticSwitchParameter(True=@n14.G, False=@n14.G) {defaultValue=true, group=SuperOverride, parameterName=SMR}
   Specular=@n14.R
   Roughness=StaticSwitchParameter(True=@n14.B, False=@n14.B) {defaultValue=true, group=SuperOverride, parameterName=SMR}
   Normal=@n15.RGB
   Emissive Color
   Opacity
   Subsurface Color=@n11
   World Position Offset
   Anisotropy
  ) {attributeSetTypes [
    #1
    #2
    #3
    #4
    #5
    #6
    #7
    #8
    #9
    #10
   ]}
  n17 = GetMaterialAttributes(None=MaterialFunctionCall!(Alpha=@n13.RGB, Materials=@n16) {materialFunction=/Game/Global/MaterialFunctions/TechArt/MF_MetalLUT_TechArt}.Result) {attributeGetTypes [#1]}
  n18 = Multiply(A=ScalarParameter() {defaultValue=1.5, group=Color_Shading, parameterName=EmissiveIntensity}, B=@n13.G)
  BlendMaterialAttributes(
   A=@n12
   B=SetMaterialAttributes!(MaterialAttributes=@n17.MaterialAttributes, Emissive Color=Multiply(A=@n17.Base Color, B=@n18)) {attributeSetTypes [#6]}
   Alpha=Add(A=@n13.R, B=@n13.R)
  )
  n19 = ObjectPositionWS()
  n20 = TransformPosition(None=@n19) {transformSourceType=TRANSFORMPOSSOURCE_World, transformType=TRANSFORMPOSSOURCE_View}
  ComponentMask(None=@n20) {b=true, g=true}
  Divide(A=TransformPosition(None=WorldPosition().XYZ) {transformSourceType=TRANSFORMPOSSOURCE_World, transformType=TRANSFORMPOSSOURCE_View}) {constB=2000}
  n21 = ScalarParameter() {defaultValue=0.02, parameterName=LocalWaveWPOFrequency}
  n22 = MaterialFunctionCall() {materialFunction=&ObjectScale}
  Divide(A=@n21, B=@n22.Scale Z)
  n23 = LinearInterpolate(Alpha=NamedRerouteUsage()) {constA=0.1}
  n24 = NamedRerouteUsage()
  n25 = VectorParameter() {group=Color_Shading, parameterName=InnerGlowColor_NoiseAdd, defaultValue=<rgba 0.2822 0.4639 0.526 1>}
  n26 = Multiply(A=ComponentMask(None=Multiply(A=@n23, B=@n24)) {r=true}, B=Multiply(A=@n25.RGB, B=@n25.RGB))
  n27 = VectorParameter() {group=Color_Shading, parameterName=BaseColorMiddleGradient, defaultValue=<rgba 0 0.1674 0.3438 1>}
  n28 = VectorParameter() {group=Color_Shading, parameterName=BaseColorBottom, defaultValue=<rgba 0 0.1554 0.1927 1>}
  n29 = Constant() {r=1}
  n30 = Constant() {r=0.5}
  n31 = Constant()
  n32 = LinearInterpolate(
   A=Multiply(A=@n28.RGB, B=@n28.RGB)
   B=Multiply(
    A=MaterialFunctionCall(Value_Morning=@n29, Value_Day=@n29, Value_Evening=@n30, Value_Night=@n31) {materialFunction=&Scalar_Time_of_Day}.Result
    B=VectorParameter() {group=Color_Shading, parameterName=BaseColorA_Scattered, defaultValue=<rgba 0.1302 2 3 1>}.RGB
   )
   Alpha=NamedRerouteUsage()
  )
  n33 = VectorParameter() {group=Color_Shading, parameterName=BaseColorTOP, defaultValue=<rgba 0 1.5 0.6438 1>}
  n34 = Constant() {r=1}
  n35 = Constant() {r=0.5}
  n36 = Constant()
  n37 = LinearInterpolate(
   A=Multiply(A=@n33.RGB, B=@n33.RGB)
   B=Multiply(
    A=MaterialFunctionCall(Value_Morning=@n34, Value_Day=@n34, Value_Evening=@n35, Value_Night=@n36) {materialFunction=&Scalar_Time_of_Day}.Result
    B=VectorParameter() {group=Color_Shading, parameterName=BaseColorScattered, defaultValue=<rgba 0.0041 0.4323 0.2735 1>}.RGB
   )
   Alpha=NamedRerouteUsage()
  )
  n38 = ScalarParameter() {defaultValue=0.35, group=Color_Shading, parameterName=HandHeight}
  n39 = Divide(A=ComponentMask(None=NamedRerouteUsage()) {b=true}, B=ScalarParameter() {defaultValue=30, group=Color_Shading, parameterName=ColorStep}) {constB=35}
  n40 = Step(Y=@n38, X=@n39)
  n41 = SmoothStep(
   Min=ScalarParameter() {group=Color_Shading, parameterName=WidthMin}
   Max=ScalarParameter() {defaultValue=0.5, group=Color_Shading, parameterName=WidthMax}
   Value=Saturate(None=Distance(A=@n39, B=@n38))
  )
  n42 = Multiply(
   A=LinearInterpolate(Alpha=NamedRerouteUsage()) {constA=1, constB=2}
   B=LinearInterpolate(A=Multiply(A=@n27.RGB, B=@n27.RGB), B=LinearInterpolate(A=@n32, B=@n37, Alpha=@n40), Alpha=@n41)
  )
  n43 = VectorParameter() {group=Color_Shading, parameterName=InnerGlowColor, defaultValue=<rgba 0 0.52 0.6979 1>}
  n44 = NamedRerouteUsage()
  n45 = ScalarParameter() {defaultValue=0.3, group=Color_Shading, parameterName=InnerGlowRadius}
  n46 = ScalarParameter() {defaultValue=-0.5, group=Color_Shading, parameterName=InnerGlowHardness}
  n47 = Saturate(None=NamedRerouteDeclaration(None=SphereMask(A=@n1, B=@n44, Radius=@n45, Hardness=@n46) {attenuationRadius=12, hardnessPercent=55}) {name=InnerGlow, nodeColor=<rgba 0.1719 0 1 1>})
  n48 = VectorParameter() {group=Color_Shading, parameterName=InnerGlowColor_Core, defaultValue=<rgba 0 0.52 0.6979 1>}
  n49 = LinearInterpolate(
   A=LinearInterpolate(A=LinearInterpolate(A=@n42, B=@n43.RGB, Alpha=@n47), B=Multiply(A=@n48.RGB, B=@n48.RGB), Alpha=NamedRerouteUsage())
   B=VectorParameter() {group=Color_Shading, parameterName=InnerScatteringColor, defaultValue=<rgba 0.023 0.057 0.276 1>}.RGB
   Alpha=NamedRerouteUsage()
  )
  n50 = VectorParameter() {group=Color_Shading, parameterName=CoreGlowColor_Add, defaultValue=<rgba 0.526 0.279 0.5044 1>}
  n51 = Multiply(A=@n50.RGB, B=@n50.RGB)
  n52 = NamedRerouteUsage()
  n53 = Add(A=Add(A=Add(A=@n26, B=@n49), B=Multiply(A=@n51, B=@n52)), B=NamedRerouteUsage())
  n54 = Multiply(A=Multiply(A=@n53, B=VectorParameter() {group=Color_Shading, parameterName=LogoColor, defaultValue=<rgba 1 1 0 1>}.RGB)) {constB=3}
  n55 = TextureObjectParameter() {group=Color_Shading, parameterName=SpriteLogo, samplerType=SAMPLERTYPE_Masks, texture=/Sprites_ZeroPointSprite/Assets/Textures/T_Sprite_Hero}
  n56 = VectorParameter() {group=UV_Screen, parameterName=TexCoordOffset, defaultValue=<rgba 0.65 0.42 0.43 0.19>}
  n57 = MaterialFunctionCall!(
   UVs=BumpOffset(Height=ScalarParameter() {defaultValue=-1.2, group=UV_Screen, parameterName=BumpOffsetLogoHeight}) {heightRatio=0.1}
   TextureIn=@n55
   UpperLeftCorner=AppendVector(A=@n56.A, B=@n56.A)
   LowerRightCorner=AppendVector(A=@n56.G, B=@n56.G)
   ExportFloat4
  ) {materialFunction=/Engine/Functions/Engine_MaterialFunctions01/Texturing/TextureCropping}
  n58 = Multiply(
   A=StaticSwitchParameter(True=LinearInterpolate(A=@n53, B=@n54, Alpha=@n57.CroppedMasked), False=@n53) {group=SuperOverride, parameterName=UseLogo}
   B=MaterialFunctionCall(
    Value_Morning=ScalarParameter() {defaultValue=1, group=Color_Shading_TOD, parameterName=TOD_Morning}
    Value_Day=ScalarParameter() {defaultValue=1, group=Color_Shading_TOD, parameterName=TOD_Day}
    Value_Evening=ScalarParameter() {defaultValue=1, group=Color_Shading_TOD, parameterName=TOD_Evening}
    Value_Night=ScalarParameter() {defaultValue=0.6, group=Color_Shading_TOD, parameterName=TOD_Night}
   ) {materialFunction=&Scalar_Time_of_Day}.Result
  )
  MaterialFunctionCall!(Color=@n58) {materialFunction=/Game/Packages/Fortress_SharedMaterials/Anime_Master_Material/MF_TODColorCurve}
  n59 = LocalPosition() {includedOffsets=ExcludeOffsets, localOrigin=InstancePreSkinning}
  n60 = ShaderStageSwitch(PixelShader=VertexInterpolator(VS=@n59.XYZ).PS, VertexShader=@n59.XYZ)
  n61 = Multiply(A=@n60) {constB=0.05}
  n62 = Abs(None=ComponentMask(None=@n61) {r=true})
  n63 = Multiply(A=SmoothStep(Value=ComponentMask(None=@n61) {b=true}) {constMax=2, constMin=1}, B=Power(Base=OneMinus(None=@n62)) {constExponent=1})
  n64 = Panner(
   Coordinate=ComponentMask(None=Multiply(A=@n60, B=@n21) {constB=0.02}) {g=true, r=true}
   Speed=AppendVector(A=ScalarParameter() {defaultValue=0.1, parameterName=CapWPOSpeedX}, B=ScalarParameter() {defaultValue=0.06, parameterName=CapWPOSpeedY})
  ) {speedX=0.1, speedY=0.01}
  n65 = Power(
   Base=TextureSample(UVs=@n64, Tex=TextureObject() {texture=/Sprites_ZeroPointSprite/Assets/Textures/Cell_Noise_T1}) {samplerType=SAMPLERTYPE_LinearGrayscale, texture=/Game/Effects/Fort_Effects/BluGlow/Violentium/Cellular_basecolor}.RGB
   Exp=ScalarParameter() {defaultValue=0.2, parameterName=CapWPONoisePower}
  )
  n66 = ScalarParameter() {defaultValue=6, parameterName=CapWPOAmount}
  n67 = MaterialFunctionCall() {materialFunction=&ObjectScale}
  n68 = Multiply(A=@n63, B=Multiply(A=ConstantBiasScale(None=@n65) {bias=-0.5, scale=2}, B=Multiply(A=@n66, B=@n67.Scale Z)))
  Multiply(A=@n68, B=VertexNormalWS())
  Multiply(A=@n22.Scale Z) {constB=0.5}
  Multiply(A=@n24, B=NamedRerouteUsage())
  Multiply(A=@n10, B=ScalarParameter() {defaultValue=1, group=UV, parameterName=DistortIntensity})
  n69 = LocalPosition() {includedOffsets=ExcludeOffsets, localOrigin=InstancePreSkinning}
  NamedRerouteDeclaration(None=ShaderStageSwitch(PixelShader=VertexInterpolator(VS=@n69.XYZ).PS, VertexShader=@n69.XYZ)) {name=PreskinnedPosition, nodeColor=<rgba 1 0.2344 0 1>}
  n70 = NamedRerouteUsage()
  n71 = VectorParameter() {group=Color_Shading, parameterName=SSC, defaultValue=<rgba 0 0.974 0.974 1>}
  n72 = Saturate(None=DotProduct(
    A=AtmosphericLightVector()
    B=MaterialFunctionCall() {materialFunction=&CameraDirectionVector}.Result
   ))
  NamedRerouteDeclaration(None=LinearInterpolate(B=Multiply(A=@n70, B=@n71.RGB), Alpha=NamedRerouteDeclaration(None=@n72) {name=SunMask, nodeColor=<rgba 0.8047 0 1 1>})) {name=SSC, nodeColor=<rgba 0 0.4141 1 1>}
  n73 = Constant()
  n74 = NamedRerouteDeclaration(None=MaterialFunctionCall(X=@n73, Y=@n73, Z=@n68) {materialFunction=&MakeFloat3}.Result) {name=CapWPO, nodeColor=<rgba 0.6406 1 0 1>}
  n75 = Multiply(A=@n74, B=Power(Base=@n63))
  n76 = VertexNormalWS()
  n77 = Transform(None=Normalize(VectorInput=Add(A=@n75, B=@n76))) {transformSourceType=TRANSFORMSOURCE_World, transformType=TRANSFORM_Tangent}
  NamedRerouteDeclaration(None=MaterialFunctionCall(BaseNormal=NamedRerouteDeclaration(None=@n77) {name=FixedNormal, nodeColor=<rgba 1 0 0.8203 1>}, AdditionalNormal=NamedRerouteUsage()) {materialFunction=&BlendAngleCorrectedNormals}.Result) {name=CorrectedNormal, nodeColor=<rgba 0.8984 1 0 1>}
  n78 = ComponentMask(None=NamedRerouteUsage()) {b=true, r=true}
  NamedRerouteDeclaration(None=SphereMask(A=Divide(A=@n78) {constB=1}, B=Constant2Vector() {g=20}) {attenuationRadius=13, hardnessPercent=15}) {name=FacialMask, nodeColor=<rgba 1 0 0.2109 1>}
  n79 = Constant()
  n80 = AppendVector(A=AppendVector(A=@n79, B=@n79), B=ScalarParameter() {defaultValue=35, group=Color_Shading, parameterName=CenterHeight})
  NamedRerouteDeclaration(None=SphereMask(
    A=@n1
    B=NamedRerouteDeclaration(None=@n80) {name=OffsetHeight, nodeColor=<rgba 1 0 0.7266 1>}
    Radius=ScalarParameter() {defaultValue=0.3, group=Color_Shading, parameterName=InnerGlowRadius02}
    Hardness=ScalarParameter() {defaultValue=-0.5, group=Color_Shading, parameterName=InnerGlowHardness02}
   ) {attenuationRadius=12, hardnessPercent=55}) {name=InnerGlow_Core, nodeColor=<rgba 0.1719 0 1 1>}
  n81 = Time()
  n82 = ScalarParameter() {defaultValue=1, parameterName=ShinningSpeed}
  n83 = ConstantBiasScale(None=Sine(None=Multiply(A=@n81, B=@n82)))
  NamedRerouteDeclaration(None=@n83) {name=ShinningAlpha, nodeColor=<rgba 0.1016 1 0 1>}
  n84 = NamedRerouteUsage()
  NamedRerouteDeclaration(None=Multiply(A=@n5, B=LinearInterpolate(Alpha=@n84) {constA=1, constB=2})) {name=CoreEM, nodeColor=<rgba 0 0.4375 1 1>}
  n85 = Constant2Vector()
  n86 = ScreenPosition()
  n87 = MaterialFunctionCall(Float2=ViewSize()) {materialFunction=&BreakOutFloat2Components}
  n88 = AppendVector(A=Constant() {r=1}, B=Divide(A=@n87.G, B=@n87.G))
  n89 = Multiply(A=ConstantBiasScale(None=Add(A=@n85, B=@n86.ViewportUV)) {bias=-0.5, scale=2}, B=@n88)
  n90 = Multiply(
   A=Constant2Vector() {g=-0.5, r=0.5}
   B=MaterialFunctionCall(World Position=@n19) {materialFunction=&TransformToClipSpace}.Clip Space XY
  ) {constA=1}
  n91 = ScalarParameter() {defaultValue=4, group=UV_Screen, parameterName=Screen Space Scale}
  n92 = MaterialFunctionCall() {materialFunction=&ObjectScale}
  n93 = MaterialFunctionCall(
   UVs=Subtract(A=ConstantBiasScale(None=@n89), B=Multiply(A=@n88, B=@n90))
   Texture Scale=Divide(A=Multiply(A=@n91, B=@n92.Scale X), B=ComponentMask(None=@n20) {b=true}) {constA=100}
  ) {materialFunction=&ScaleUVsByCenter}
  n94 = Constant2Vector() {g=0.3, r=0.4}
  NamedRerouteDeclaration(None=Multiply(A=Multiply(A=@n93.UVs, B=@n94), B=MaterialFunctionCall() {materialFunction=/Engine/Functions/Engine_MaterialFunctions02/Utility/FOV}."FOV ")) {name=ScreenUV, nodeColor=<rgba 1 0 0.1172 1>}
  NamedRerouteDeclaration(None=@n41) {name=HandAlpha, nodeColor=<rgba 0.875 0 1 1>}
  NamedRerouteDeclaration(None=AppendVector(A=AppendVector(A=@n40, B=@n41), B=Constant())) {name=HandMask, nodeColor=<rgba 1 0.0234 0 1>}
  n95 = TextureObjectParameter() {group=Color_Shading, parameterName=Pattern, samplerType=SAMPLERTYPE_Masks, texture=&T_BlackMask}
  n96 = Constant()
  n97 = MaterialFunctionCall(
   Texture=@n95
   UseBottomLayer=StaticBool() {value=true}
   Depth=ScalarParameter() {defaultValue=10, group=UV, parameterName=Zero_02_Depth, sortPriority=61}
   Depth_Z=ScalarParameter() {defaultValue=25, group=UV, parameterName=Zero_02_Depth_Z, sortPriority=61}
   Tile=ScalarParameter() {defaultValue=5, group=UV, parameterName=Zero_02_Tile, sortPriority=62}
   BlendContrast_Z=ScalarParameter() {defaultValue=0.8, group=UV, parameterName=Zero_02_BlendContrastZ, sortPriority=63}
   BlendContrast=ScalarParameter() {defaultValue=0.4, group=UV, parameterName=Zero_02_BlendContrast, sortPriority=63}
   SpeedXY=ComponentMask(None=VectorParameter() {group=UV, parameterName=Zero_02_Speed, sortPriority=64, defaultValue=<rgba 0 0 0 1>}.RGB) {g=true, r=true}
   ActorOffset=VectorParameter() {group=UV, parameterName=Zero_02_Offset, sortPriority=64, defaultValue=<rgba 0 0 20 1>}.RGB
   Distort=@n96
  ) {materialFunction=&MF_TriplanarVirtualBox_WhisperRemedySweet}
  NamedRerouteDeclaration(None=ComponentMask(None=@n97.Result) {g=true}) {name=ColorNoise, nodeColor=<rgba 0.7813 0 1 1>}
  n98 = MaterialFunctionCall(Normal Vector=@n8, Power=ScalarParameter() {defaultValue=3, group=Color_Shading_Fresnel, parameterName=OutGlowFresnelPow}) {materialFunction=&Fresnel_Function}
  n99 = VectorParameter() {group=Color_Shading_Fresnel, parameterName=OutGlowFresnelColor, defaultValue=<rgba 0 0.9323 0.6506 1>}
  NamedRerouteDeclaration(None=Multiply(A=@n98.Result, B=Multiply(A=@n99.RGB, B=@n99.RGB))) {name=OuterGlow, nodeColor=<rgba 1 0.5156 0 1>}
  n100 = StaticBool()
  n101 = ScalarParameter() {defaultValue=10, group=UV, parameterName=Zero_01_Depth, sortPriority=61}
  n102 = ScalarParameter() {defaultValue=25, group=UV, parameterName=Zero_01_Depth_Z, sortPriority=61}
  n103 = ScalarParameter() {defaultValue=5, group=UV, parameterName=Zero_01_Tile, sortPriority=62}
  n104 = ScalarParameter() {defaultValue=0.8, group=UV, parameterName=Zero_01_BlendContrastZ, sortPriority=63}
  n105 = ScalarParameter() {defaultValue=0.4, group=UV, parameterName=Zero_01_BlendContrast, sortPriority=63}
  n106 = ComponentMask(None=VectorParameter() {group=UV, parameterName=Zero_01_Speed, sortPriority=64, defaultValue=<rgba 0 0 0 1>}.RGB) {g=true, r=true}
  n107 = VectorParameter() {group=UV, parameterName=Zero_01_Offset, sortPriority=64, defaultValue=<rgba 0 0 20 1>}
  n108 = ComponentMask(None=Add(
    A=MaterialFunctionCall(
     Texture=@n95
     UseBottomLayer=@n100
     Depth=@n101
     Depth_Z=@n102
     Tile=@n103
     BlendContrast_Z=@n104
     BlendContrast=@n105
     SpeedXY=@n106
     ActorOffset=@n107.RGB
     Distort=@n96
    ) {materialFunction=&MF_TriplanarVirtualBox_WhisperRemedySweet}.Result
    B=Power(Base=@n97.Result)
   )) {r=true}
  n109 = Power(Base=@n108) {constExponent=2.2}
  n110 = Multiply(A=@n108) {constB=0.1}
  NamedRerouteDeclaration(None=Multiply(A=Add(A=@n109, B=@n110), B=ScalarParameter() {defaultValue=3, parameterName=GlowNoiseIntensity})) {name=InnerGlowNoise, nodeColor=<rgba 0.3125 0 1 1>}
  n111 = ScalarParameter() {defaultValue=1, parameterName=EyeAdaptionInverse}
  NamedRerouteDeclaration(None=StaticSwitchParameter(True=NamedRerouteUsage(), False=EyeAdaptationInverse(LightValueInput=@n58, AlphaInput=@n111).EyeAdaptationInverse) {group=Color_Shading, parameterName="[Debug]HandMask"}) {name=Emissive, nodeColor=<rgba 0.7344 1 0 1>}
  n112 = NamedRerouteUsage()
  n113 = WorldPosition() {worldPositionShaderOffset=WPT_ExcludeAllShaderOffsets}
  n114 = ScalarParameter() {defaultValue=0.025, parameterName=WorldWaveWPOFrequency}
  n115 = Subtract(A=Frac(None=Multiply(A=@n113.XYZ, B=@n114) {constB=0.0025})) {constB=0.5}
  n116 = MaterialFunctionCall(Float3=Abs(None=Multiply(A=@n115) {constB=2})) {materialFunction=&BreakOutFloat3Components}
  n117 = Add(A=Add(A=@n116.R, B=@n116.R), B=@n116.B)
  n118 = Multiply(A=Time()) {constB=0.2}
  n119 = Sine(None=Subtract(A=Subtract(A=@n117) {constB=2}, B=Frac(None=@n118)))
  n120 = Multiply(A=Multiply(A=Add(A=@n119)) {constB=0.5}, B=ScalarParameter() {parameterName=WPOWorldScale})
  n121 = MaterialFunctionCall(Float3=Multiply(
    A=LocalPosition() {includedOffsets=ExcludeOffsets, localOrigin=InstancePreSkinning}.XYZ
    B=ScalarParameter() {defaultValue=0.02, parameterName=LocalWaveWPOFrequency}
   ) {constB=0.02}) {materialFunction=&BreakOutFloat3Components}
  n122 = Add(A=@n121.R, B=@n121.R)
  n123 = Subtract(A=Time(), B=ScalarParameter() {parameterName=WPO_random_seed})
  n124 = Subtract(A=Subtract(A=Add(A=@n122, B=@n121.B)) {constB=2}, B=Frac(None=Multiply(A=@n123) {constB=0.75}))
  n125 = Multiply(A=Add(A=Sine(None=@n124))) {constB=0.5}
  n126 = Multiply(A=Multiply(A=VertexNormalWS(), B=Add(A=@n120, B=@n125) {constB=0}), B=ScalarParameter() {defaultValue=0.5, parameterName=WPO_FinalScale})
  NamedRerouteDeclaration(None=LinearInterpolate(A=Add(A=@n112, B=@n126), Alpha=NamedRerouteUsage()) {constB=0}) {name=WPO, nodeColor=<rgba 0.9453 0 1 1>}
  n127 = AtmosphericLightVector()
  n128 = PixelNormalWS()
  n129 = DotProduct(A=@n127, B=@n128)
  n130 = Subtract(A=Multiply(A=Multiply(A=@n129) {constB=2}, B=@n128), B=@n127)
  n131 = MaterialFunctionCall() {materialFunction=&CameraDirectionVector}
  n132 = Max(A=DotProduct(A=Normalize(VectorInput=@n130), B=Multiply(A=@n131.Result) {constB=-1})) {constB=0}
  n133 = ScalarParameter() {defaultValue=2, parameterName=InnerScatteringPower}
  NamedRerouteDeclaration(None=Multiply(A=Power(Base=@n132, Exp=@n133), B=ScalarParameter() {defaultValue=0.7, parameterName=InnerScatteringScale}) {constB=0.4}) {name=InnerScatter, nodeColor=<rgba 1 0 0.1406 1>}
  n134 = Constant3Vector() {constant=<rgba 0 0 1 0>}
  n135 = NamedRerouteUsage()
  n136 = Multiply(A=NamedRerouteUsage()) {constB=10}
  n137 = Add(A=@n135, B=Multiply(A=Saturate(None=@n136), B=ScalarParameter() {defaultValue=-0.01, parameterName=BubbleEdgeWarp}))
  n138 = BumpOffset(
   Coordinate=Multiply(A=Panner(Coordinate=@n137) {speedY=0.15}, B=ScalarParameter() {defaultValue=1.5, parameterName=BubbleTilingR}) {constB=3}
   Height=ScalarParameter() {defaultValue=0.05, parameterName=BubbleOffsetR}
  ) {heightRatio=0.1}
  n139 = TextureObjectParameter() {
   group=Slurp
   parameterName=BubbleMask
   samplerType=SAMPLERTYPE_Masks
   texture=/Game/Characters/Player/Male/Large/Bodies/M_LRG_Slurp_Monster/Textures/T_SlurpBubble
  }
  n140 = TextureSample(UVs=@n138, Tex=@n139)
  n141 = TextureObjectParameter() {
   group=Slurp
   parameterName=BubbleNormal
   samplerType=SAMPLERTYPE_Normal
   texture=/Game/Characters/Player/Male/Large/Bodies/M_LRG_Slurp_Monster/Textures/T_SlurpBubble_N
  }
  n142 = Multiply(A=Panner(Coordinate=@n135) {speedY=0.1}, B=ScalarParameter() {defaultValue=1, parameterName=BubbleTilingG}) {constB=3}
  n143 = TextureSample(UVs=@n142, Tex=@n139)
  n144 = Add(A=Multiply(A=@n140.R, B=TextureSample(UVs=@n138, Tex=@n141).RGB), B=Multiply(A=@n143.G, B=TextureSample(UVs=@n142, Tex=@n141).RGB))
  n145 = Add(A=@n140.R, B=@n143.G)
  n146 = Multiply(A=LinearInterpolate(Alpha=NamedRerouteUsage())) {constB=15}
  NamedRerouteDeclaration(None=MaterialFunctionCall(Normal=LinearInterpolate(A=@n134, B=@n144, Alpha=@n145), Flatness=Saturate(None=@n146)) {materialFunction=&FlattenNormal}.Result) {name=BubbleNormal, nodeColor=<rgba 0 1 0.7891 1>}
  NamedRerouteDeclaration(None=@n98.Result) {name=OuterFresnel, nodeColor=<rgba 0 0.1797 1 1>}
  NamedRerouteUsage()
  NamedRerouteUsage()
  NamedRerouteUsage()
  n147 = MaterialFunctionCall!(UVs=@n64, Normal=@n74) {materialFunction=/Engine/Functions/Engine_MaterialFunctions01/Texturing/DeriveTangentBasis}
  n148 = VertexNormalWS()
  Normalize(VectorInput=MaterialFunctionCall(X=@n147.Tangent, Y=CrossProduct(A=@n147.Tangent, B=@n148), Z=@n148) {materialFunction=&MakeFloat3}.Result)
  n149 = PixelDepth()
  OneMinus(None=Saturate(None=Divide(A=@n149.R) {constB=100}))
  PixelNormalWS()
  n150 = VertexInterpolator(VS=PreSkinnedPosition())
  n151 = VectorParameter() {parameterName=EyeCenter, defaultValue=<rgba 0 0 0 1>}
  n152 = Divide(
   A=Length(None=Subtract(A=@n150.PS, B=@n151.RGB)) {is=class MaterialExpressionLength}
   B=ScalarParameter() {defaultValue=15, parameterName=EyeRadius}
  )
  Power(Base=OneMinus(None=Saturate(None=@n152))) {constExponent=2.2}
  RayTracingQualitySwitch()
  SmoothStep(Value=@n83) {constMax=0.66, constMin=0.33}
  TextureCoordinate()
  n153 = Constant2Vector() {g=0.5, r=1}
  TextureSampleParameter2D(UVs=Panner(Coordinate=Multiply(B=@n153)) {speedX=-0.05, speedY=-0.15}) {parameterName=TexWaves, samplerType=SAMPLERTYPE_LinearGrayscale, texture=/Game/Effects/Fort_Effects/Textures/Tile/T_Caustic_Wind_Diagonal}
  VectorParameter() {group=Color_Shading, parameterName=MaskColorEM, defaultValue=<rgba 0 0 0 1>}
 ]
}
