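# Character cape material for the WhisperRemedy Sprite, built on shared CharacterParent PreFX and PostFX functions with per-expression shading models and skeletal mesh, clothing and Nanite usage.
# The PreFX function supplies the base material attributes and a Tech Art mask.
# Emissive color is set to base color multiplied by that mask's red channel and an EmissiveIntensity parameter, making masked areas glow in their own color.
# The modified attributes then feed the PostFX function.
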Material /Sprites_WhisperRemedySweetSprite/Assets/Materials/M_Sprite_WhisperRemedy_Cape
graph {
 properties {
  bUseMaterialAttributes=true
  bUsedWithClothing=true
  bUsedWithMeshDeformer=true
  bUsedWithMorphTargets=true
  bUsedWithNanite=true
  bUsedWithNiagaraMeshParticles=true
  bUsedWithSkeletalMesh=true
  pixelDepthOffsetMode=PDOM_Legacy
  shadingModel=MSM_FromMaterialExpression
  usedShadingModels=Default Lit | Subsurface | Clear Coat | Subsurface Profile | Cloth
 }
 nodes [
  n1 = MaterialFunctionCall() {materialFunction=&MF_CharacterParent_2023_PreFX}
  n2 = GetMaterialAttributes(None=@n1.Result) {attributeGetTypes [#1]}
  n3 = Multiply(
   A=ScalarParameter() {defaultValue=1.5, group=Color_Shading, parameterName=EmissiveIntensity}
   B=ComponentMask(None=@n1.Tech Art Mask) {r=true}
  )
  MaterialFunctionCall(
   Material=SetMaterialAttributes(MaterialAttributes=@n2.MaterialAttributes, Emissive Color=Multiply(A=@n2.Base Color, B=@n3)) {attributeSetTypes [#2]}
   M=@n1.M
   Pre Skinned Local Position=@n1.M
   Distance Blend=@n1.M
   Fresnel=@n1.M
  ) {materialFunction=&MF_CharacterParent_2023_PostFX}
 ]
}
