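# Animation Blueprint for the Whisper Remedy Sweet Sprite creature, covering locomotion, fleeing and heat-vision aim and fire.
# A state machine uses aliases and a conduit to link Idle, Walk, Run, Shocked and Stop with Targeting and firing states.
# Idle picks fleeing, equipped, cooldown or backpack poses by bool, with optional spring bones.
# The cached result gets a pitch-driven additive eye blendspace, layered from spine_2; the event graph only sets the fleeing flag.
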AnimBlueprint /Sprites_WhisperRemedySweetSprite/Assets/Meshes/Sprite_WhisperRemedySweet_Creature_Skeleton_AnimBlueprint
target_skeleton=/Sprites/Assets/Meshes/Sprite_Creature_Skeleton
graphs [
 AnimGraph {nodes [
   n1 = Blendspace Player 'Sprite_HeatVision_Eyes_Adjustment_BS'!(Y=Get ReturnValueYPitch() -> (ReturnValueYPitch:float)) -> (Pose) {blendSpace=/Sprites_WhisperRemedySweetSprite/Assets/Animation/HeatVision/Sprite_HeatVision_Eyes_Adjustment_BS}
   n2 = Use cached pose 'PreFaceAdditivePose'()
   Output Pose(Result=Layered blend per bone(BasePose=Use cached pose 'PreFaceAdditivePose'(), BlendPoses_0=Apply Additive(Base=@n2, Additive=@n1) {alphaBoolBlend {blendOption=Linear}}) {layerSetup [{branchFilters [{boneName=spine_2}]}]})
   AnimGraphNode_StateMachine_0 = Anim States() {is=title AnimGraphNode_StateMachine}
   Save cached pose 'PreFaceAdditivePose'(Pose=@AnimGraphNode_StateMachine_0) {cacheName=PreFaceAdditivePose}
  ]}
 Anim States {
  entry=Idle
  path=AnimGraph.AnimGraphNode_StateMachine_0.Anim States
  ids=id
  nodes [
   Alias() {is=title AnimStateAliasNode, also {aliases [Idle,Stop]}}
   ToStops() {is=title AnimStateAliasNode, also {aliases [Run,Shocked,Walk]}}
   IsFleeing() {is=title AnimStateAliasNode, also {aliases [Idle,Walk]}}
   EarlyExit_1_1() {is=title AnimStateAliasNode, also {aliases [Loop]}}
   IntoFire_From_Idle() {is=title AnimStateAliasNode, also {aliases [Idle]}}
   IntoFire() {is=title AnimStateAliasNode, also {aliases [Targeting]}}
   AnimStateConduitNode_0 = Conduit()
   AnimStateNode_0 = Idle() {
    stateEntered {guid=#1}
    stateFullyBlended {guid=#2}
    stateLeft {guid=#3}
   }
   AnimStateNode_1 = Walk() {
    is=title AnimStateNode
    stateEntered {guid=#4}
    stateFullyBlended {guid=#5}
    stateLeft {guid=#6}
   }
   AnimStateNode_12 = Intro() {is=title AnimStateNode}
   AnimStateNode_2 = Shocked() {is=title AnimStateNode}
   AnimStateNode_3 = Stop() {is=title AnimStateNode}
   AnimStateNode_4 = Run() {
    is=title AnimStateNode
    stateEntered {guid=#7}
    stateFullyBlended {guid=#8}
    stateLeft {guid=#9}
   }
   AnimStateNode_5 = Targeting() {
    is=title AnimStateNode
    stateEntered {guid=#10}
    stateFullyBlended {guid=#11}
    stateLeft {guid=#12}
   }
   AnimStateNode_7 = Loop() {is=title AnimStateNode}
   AnimStateTransitionNode_0 = AnimStateTransitionNode()
   AnimStateTransitionNode_1 = AnimStateTransitionNode()
   AnimStateTransitionNode_10 = AnimStateTransitionNode()
   AnimStateTransitionNode_11 = AnimStateTransitionNode() {crossfadeDuration=0.1}
   AnimStateTransitionNode_12 = AnimStateTransitionNode() {bAutomaticRuleBasedOnSequencePlayerInState=true}
   AnimStateTransitionNode_14 = AnimStateTransitionNode()
   AnimStateTransitionNode_16 = AnimStateTransitionNode()
   AnimStateTransitionNode_17 = AnimStateTransitionNode()
   AnimStateTransitionNode_2 = AnimStateTransitionNode()
   AnimStateTransitionNode_23 = AnimStateTransitionNode()
   AnimStateTransitionNode_24 = AnimStateTransitionNode()
   AnimStateTransitionNode_25 = AnimStateTransitionNode()
   AnimStateTransitionNode_26 = AnimStateTransitionNode()
   AnimStateTransitionNode_27 = AnimStateTransitionNode() {bAutomaticRuleBasedOnSequencePlayerInState=true}
   AnimStateTransitionNode_6 = AnimStateTransitionNode()
   AnimStateTransitionNode_8 = AnimStateTransitionNode()
  ]
  links [
   {
    dst=Walk
    rule=AnimGraph.AnimGraphNode_StateMachine_0.Anim States.AnimStateTransitionNode_14.AnimationTransitionGraph_0
    src=Alias
    via=AnimStateTransitionNode_14
   }
   {
    dst=Idle
    rule=AnimGraph.AnimGraphNode_StateMachine_0.Anim States.AnimStateTransitionNode_17.Transition
    src=Conduit
    via=AnimStateTransitionNode_17
   }
   {
    dst=Targeting
    rule=AnimGraph.AnimGraphNode_StateMachine_0.Anim States.AnimStateTransitionNode_16.Transition
    src=Conduit
    via=AnimStateTransitionNode_16
   }
   {
    dst=Conduit
    rule=AnimGraph.AnimGraphNode_StateMachine_0.Anim States.AnimStateTransitionNode_23.Transition
    src=EarlyExit_1_1
    via=AnimStateTransitionNode_23
   }
   {
    dst=Targeting
    rule=AnimGraph.AnimGraphNode_StateMachine_0.Anim States.AnimStateTransitionNode_25.Transition
    src=Idle
    via=AnimStateTransitionNode_25
   }
   {
    dst=Loop
    rule=AnimGraph.AnimGraphNode_StateMachine_0.Anim States.AnimStateTransitionNode_10.Transition
    src=IntoFire
    via=AnimStateTransitionNode_10
   }
   {
    dst=Intro
    rule=AnimGraph.AnimGraphNode_StateMachine_0.Anim States.AnimStateTransitionNode_26.Transition
    src=IntoFire_From_Idle
    via=AnimStateTransitionNode_26
   }
   {
    dst=Loop
    rule=AnimGraph.AnimGraphNode_StateMachine_0.Anim States.AnimStateTransitionNode_27.Transition
    src=Intro
    via=AnimStateTransitionNode_27
   }
   {
    dst=Shocked
    rule=AnimGraph.AnimGraphNode_StateMachine_0.Anim States.AnimStateTransitionNode_8.AnimationTransitionGraph_0
    src=IsFleeing
    via=AnimStateTransitionNode_8
   }
   {
    dst=Walk
    rule=AnimGraph.AnimGraphNode_StateMachine_0.Anim States.AnimStateTransitionNode_6.AnimationTransitionGraph_0
    src=Run
    via=AnimStateTransitionNode_6
   }
   {
    dst=Idle
    rule=AnimGraph.AnimGraphNode_StateMachine_0.Anim States.AnimStateTransitionNode_1.AnimationTransitionGraph_0
    src=Shocked
    via=AnimStateTransitionNode_1
   }
   {
    dst=Run
    rule=AnimGraph.AnimGraphNode_StateMachine_0.Anim States.AnimStateTransitionNode_0.AnimationTransitionGraph_0
    src=Shocked
    via=AnimStateTransitionNode_0
   }
   {
    dst=Walk
    rule=AnimGraph.AnimGraphNode_StateMachine_0.Anim States.AnimStateTransitionNode_2.AnimationTransitionGraph_0
    src=Shocked
    via=AnimStateTransitionNode_2
   }
   {
    dst=Idle
    rule=AnimGraph.AnimGraphNode_StateMachine_0.Anim States.AnimStateTransitionNode_12.AnimationTransitionGraph_0
    src=Stop
    via=AnimStateTransitionNode_12
   }
   {
    dst=Idle
    rule=AnimGraph.AnimGraphNode_StateMachine_0.Anim States.AnimStateTransitionNode_24.Transition
    src=Targeting
    via=AnimStateTransitionNode_24
   }
   {
    dst=Stop
    rule=AnimGraph.AnimGraphNode_StateMachine_0.Anim States.AnimStateTransitionNode_11.AnimationTransitionGraph_0
    src=ToStops
    via=AnimStateTransitionNode_11
   }
  ]
 }
 Conduit {path=AnimGraph.AnimGraphNode_StateMachine_0.Anim States.AnimStateConduitNode_0.Conduit, nodes [Result(bCanEnterTransition=true)]}
 Idle {
  path=AnimGraph.AnimGraphNode_StateMachine_0.Anim States.AnimStateNode_0.Idle
  nodes [
   n1 = Spring controller - Bone: spine_1!(
    ComponentPose:Component Space Pose Link Structure=Local To Component(LocalPose=Sequence Player 'Backpack_Idle_Water'() {sequence=/Sprites_WhisperRemedySweetSprite/Assets/Animation/Backpack_Idle_Water})
    Alpha:float=0.05
   ) -> (Pose:Component Space Pose Link Structure) {
    is=title AnimGraphNode_SpringBone
    bLimitDisplacement=true
    errorResetThresh=500000
    maxDisplacement=750
    alphaBoolBlend {blendOption=Linear}
    springBone {boneName=spine_1}
   }
   n2 = Spring controller - Bone: pelvis!(ComponentPose:Component Space Pose Link Structure=@n1, Alpha:float=0.05) -> (Pose:Component Space Pose Link Structure) {
    is=title AnimGraphNode_SpringBone
    bLimitDisplacement=true
    errorResetThresh=500000
    maxDisplacement=750
    alphaBoolBlend {blendOption=Linear}
    springBone {boneName=pelvis}
   }
   n3 = Blend Poses by bool(
    BlendPose_0=Sequence Player 'Backpack_Idle_Water'() {sequence=&Backpack_Idle_Water}
    BlendPose_1=Component To Local(ComponentPose=@n2)
    bActiveValue=Get bDisableSprings() -> (bDisableSprings:bool)
   ) {blendType=Linear}
   n4 = Sequence Player 'Neutral_Idle_Sprite_Water'() {sequence=/Sprites_WhisperRemedySweetSprite/Assets/Animation/Neutral_Idle_Sprite_Water}
   n5 = Get bIsBackpack_Idle() -> (bIsBackpack_Idle:bool)
   n6 = Blend Poses by bool(
    BlendPose_0=Sequence Player 'IdleEquiped_Sprite_Water'() {sequence=/Sprites_WhisperRemedySweetSprite/Assets/Animation/IdleEquiped_Sprite_Water}
    BlendPose_1=Blend Poses by bool(
     BlendPose_0=Sequence Player 'Sleep_Water_Sprite_Water'() {sequence=/Sprites_WhisperRemedySweetSprite/Assets/Animation/Sleep_Water_Sprite_Water}
     BlendPose_1=Blend Poses by bool(BlendPose_0=@n3, BlendPose_1=@n4, bActiveValue=@n5) {blendType=Linear}
     bActiveValue=Get bIsCooldown() -> (bIsCooldown:bool)
    ) {blendType=Linear}
    bActiveValue=Get bIsEquipped_Idle() -> (bIsEquipped_Idle:bool)
   ) {blendType=Linear}
   n7 = Sequence Player 'JoyfulWalk_Sprite_Water'() {sequence=/Sprites_WhisperRemedySweetSprite/Assets/Animation/JoyfulWalk_Sprite_Water}
   n8 = Get bIsFleeing() -> (bIsFleeing:bool)
   Output Animation Pose(Result=Slot 'DefaultSlot'(Source=Blend Poses by bool(BlendPose_0=@n7, BlendPose_1=@n6, bActiveValue=@n8) {blendType=Linear}))
  ]
 }
 Walk {
  path=AnimGraph.AnimGraphNode_StateMachine_0.Anim States.AnimStateNode_1.Walk
  nodes [Output Animation Pose(Result=Sequence Player 'Walk_Sprite_Water'!(PlayRate:float=Get WalkPlayrate() -> (WalkPlayrate:double)) -> (Pose) {sequence=/Sprites_WhisperRemedySweetSprite/Assets/Animation/Walk_Sprite_Water})]
 }
 Intro {
  path=AnimGraph.AnimGraphNode_StateMachine_0.Anim States.AnimStateNode_12.Intro
  nodes [Output Animation Pose(Result=Sequence Player 'Sprite_Heatvision_Fire_Intro_WEP'() {bLoopAnimation=false, sequence=/Sprites_WhisperRemedySweetSprite/Assets/Animation/HeatVision/Sprite_Heatvision_Fire_Intro_WEP})]
 }
 Shocked {
  path=AnimGraph.AnimGraphNode_StateMachine_0.Anim States.AnimStateNode_2.Shocked
  nodes [Output Animation Pose(Result=Sequence Player 'Shock_Sprite_Water'() {bLoopAnimation=false, sequence=/Sprites_WhisperRemedySweetSprite/Assets/Animation/Shock_Sprite_Water})]
 }
 Stop {
  path=AnimGraph.AnimGraphNode_StateMachine_0.Anim States.AnimStateNode_3.Stop
  nodes [
   n1 = Sequence Player 'Settle_Water_Sprite_Water'!(StartPosition:float=0.96) -> (Pose) {bLoopAnimation=false, sequence=/Sprites_WhisperRemedySweetSprite/Assets/Animation/Settle_Water_Sprite_Water, startPosition=0.96}
   Output Animation Pose(Result=@n1)
  ]
 }
 Run {
  path=AnimGraph.AnimGraphNode_StateMachine_0.Anim States.AnimStateNode_4.Run
  nodes [Output Animation Pose(Result=Sequence Player 'Run_Sprite'!(PlayRate:float=Get RunPlayrate() -> (RunPlayrate:double)) -> (Pose) {sequence=/Sprites_WhisperRemedySweetSprite/Assets/Animation/Run_Sprite})]
 }
 Targeting {
  path=AnimGraph.AnimGraphNode_StateMachine_0.Anim States.AnimStateNode_5.Targeting
  nodes [
   Sequence Player 'Sprite_Heatvision_Fire_Loop_WEP'() {sequence=/Sprites_WhisperRemedySweetSprite/Assets/Animation/HeatVision/Sprite_Heatvision_Fire_Loop_WEP}
   Output Animation Pose(Result=Sequence Player 'Sprite_Heatvision_Aim_Loop_WEP'() {sequence=/Sprites_WhisperRemedySweetSprite/Assets/Animation/HeatVision/Sprite_Heatvision_Aim_Loop_WEP})
  ]
 }
 Loop {
  path=AnimGraph.AnimGraphNode_StateMachine_0.Anim States.AnimStateNode_7.Loop
  nodes [Output Animation Pose(Result=Sequence Player 'Sprite_Heatvision_Fire_Loop_WEP'() {sequence=&Sprite_Heatvision_Fire_Loop_WEP})]
 }
 AnimationTransitionGraph_0 {
  path=AnimGraph.AnimGraphNode_StateMachine_0.Anim States.AnimStateTransitionNode_0.AnimationTransitionGraph_0
  nodes [
   Result(bCanEnterTransition=Get bEnterRun() -> (bEnterRun:bool))
   "float > float"(A="Current Time (Shock_Sprite_Water)"!() -> (ReturnValue:float) {is=title K2Node_AnimGetter}, B=0.75)
  ]
 }
 AnimationTransitionGraph_0 {
  path=AnimGraph.AnimGraphNode_StateMachine_0.Anim States.AnimStateTransitionNode_1.AnimationTransitionGraph_0
  nodes [Result(bCanEnterTransition=Get bEnterIdle() -> (bEnterIdle:bool))]
 }
 Transition {
  path=AnimGraph.AnimGraphNode_StateMachine_0.Anim States.AnimStateTransitionNode_10.Transition
  nodes [Result(bCanEnterTransition=Get bRaiseSprite() -> (bRaiseSprite:bool))]
 }
 AnimationTransitionGraph_0 {
  path=AnimGraph.AnimGraphNode_StateMachine_0.Anim States.AnimStateTransitionNode_11.AnimationTransitionGraph_0
  nodes [Result(bCanEnterTransition=NOT Boolean(A=Get bIsMoving() -> (bIsMoving:bool)))]
 }
 AnimationTransitionGraph_0 {path=AnimGraph.AnimGraphNode_StateMachine_0.Anim States.AnimStateTransitionNode_12.AnimationTransitionGraph_0, nodes [Result()]}
 AnimationTransitionGraph_0 {
  path=AnimGraph.AnimGraphNode_StateMachine_0.Anim States.AnimStateTransitionNode_14.AnimationTransitionGraph_0
  nodes [Result(bCanEnterTransition=Get bIsMoving())]
 }
 Transition {
  path=AnimGraph.AnimGraphNode_StateMachine_0.Anim States.AnimStateTransitionNode_16.Transition
  nodes [Result(bCanEnterTransition=Get bIsTargeting())]
 }
 Transition {
  path=AnimGraph.AnimGraphNode_StateMachine_0.Anim States.AnimStateTransitionNode_17.Transition
  nodes [Result(bCanEnterTransition=NOT Boolean(A=Get bIsTargeting()))]
 }
 AnimationTransitionGraph_0 {
  path=AnimGraph.AnimGraphNode_StateMachine_0.Anim States.AnimStateTransitionNode_2.AnimationTransitionGraph_0
  nodes [Result(bCanEnterTransition=Get bEnterWalk() -> (bEnterWalk:bool))]
 }
 Transition {
  path=AnimGraph.AnimGraphNode_StateMachine_0.Anim States.AnimStateTransitionNode_23.Transition
  nodes [Result(bCanEnterTransition=NOT Boolean(A=Get bRaiseSprite()))]
 }
 Transition {
  path=AnimGraph.AnimGraphNode_StateMachine_0.Anim States.AnimStateTransitionNode_24.Transition
  nodes [Result(bCanEnterTransition=NOT Boolean(A=Get bIsTargeting()))]
 }
 Transition {
  path=AnimGraph.AnimGraphNode_StateMachine_0.Anim States.AnimStateTransitionNode_25.Transition
  nodes [Result(bCanEnterTransition=Get FP_IdleToTargeting() -> (FP_IdleToTargeting:bool))]
 }
 Transition {
  path=AnimGraph.AnimGraphNode_StateMachine_0.Anim States.AnimStateTransitionNode_26.Transition
  nodes [Result(bCanEnterTransition=Get bRaiseSprite())]
 }
 Transition {path=AnimGraph.AnimGraphNode_StateMachine_0.Anim States.AnimStateTransitionNode_27.Transition, nodes [Result()]}
 AnimationTransitionGraph_0 {
  path=AnimGraph.AnimGraphNode_StateMachine_0.Anim States.AnimStateTransitionNode_6.AnimationTransitionGraph_0
  nodes [Result(bCanEnterTransition=NOT Boolean(A=Get bIsFleeing()))]
 }
 AnimationTransitionGraph_0 {
  path=AnimGraph.AnimGraphNode_StateMachine_0.Anim States.AnimStateTransitionNode_8.AnimationTransitionGraph_0
  nodes [Result(bCanEnterTransition=Get bIsFleeing())]
 }
 EventGraph {nodes [
   n1 = Event SetFleeing!() -> (OutputDelegate, then, NewValue:bool) {is=title K2Node_Event}
   > Set bIsFleeing(bIsFleeing:bool=@n1.NewValue) -> (Output_Get:bool)
  ]}
]
variables []
